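# Translucent, two-sided, default-lit particle material for flower petals in the 60s Hippie pickaxe's effects, enabled for Niagara mesh particles.
# It samples the Hippie glider petal texture through a SubUV node.
# The texture's red channel multiplies particle color as a grayscale tint mask, and its alpha multiplies particle alpha.
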Material /TidalCraneHoagie/Weapons/FORT_Melee/Pickaxe_60sHippie/Mesh/FX/M_FlowerPetals
graph {
 properties {
  bUsedWithNiagaraMeshParticles=true
  blendMode=BLEND_Translucent
  pixelDepthOffsetMode=PDOM_Legacy
  twoSided=true
  usedShadingModels=Default Lit
 }
 nodes [
  n1 = ParticleSubUV!(UVs, Tex, Apply View MipBias) {texture=&/T_HippieGlider_Petal_D}
  n2 = ParticleColor()
  Multiply(A=@n1.R, B=@n2.RGB)
  Multiply(A=@n1.A, B=@n2.A)
 ]
}
